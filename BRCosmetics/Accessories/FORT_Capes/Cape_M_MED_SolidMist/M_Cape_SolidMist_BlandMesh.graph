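# Translucent, unlit, two-sided skeletal-mesh material for the SolidMist cape, rendering warm orange, streaming mist.
# Smoke, caustic, plasma and noise textures pan along U, some distorted by a scrolling normal noise, forming tinted emissive layers scaled by FinalColor.
# UV gradient and sine masks shape the tail and side edges, and the masked result drives depth-faded opacity.
# World position offset sums a local vector, a U-gradient push and a camera offset.
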Material /BRCosmetics/Accessories/FORT_Capes/Cape_M_MED_SolidMist/M_Cape_SolidMist_BlandMesh
graph {
 properties {
  bAutomaticallySetUsageInEditor=false
  bUsedWithMeshDeformer=true
  bUsedWithSkeletalMesh=true
  blendMode=BLEND_Translucent
  shadingModel=MSM_Unlit
  twoSided=true
  usedShadingModels=Unlit
 }
 nodes [
  n1 = Power(Base=ComponentMask(None=TextureCoordinate()) {r=true}) {constExponent=0.6}
  n2 = Saturate(None=Multiply(A=Multiply(A=@n1)) {constB=1.3})
  n3 = ScalarParameter() {parameterName=WPOIntensity}
  n4 = ScalarParameter() {defaultValue=3, parameterName="Offset Amount (S)"}
  Add(
   A=Transform(None=VectorParameter() {parameterName=WPO Offset, defaultValue=<rgba 0 -3 -3 0>}.RGB) {transformSourceType=TRANSFORMSOURCE_Local}
   B=Add(
    A=Multiply(A=@n2, B=@n3)
    B=MaterialFunctionCall(Offset Amount=@n4) {materialFunction=/Engine/Functions/Engine_MaterialFunctions02/WorldPositionOffset/CameraOffset}.Camera Offset
   ) {constB=0}
  )
  n5 = TextureCoordinate() {uTiling=0.6, vTiling=5}
  n6 = TextureCoordinate()
  n7 = AppendVector(A=Constant() {r=0.8}, B=Constant() {r=2})
  n8 = TextureSample(UVs=Panner(Coordinate=Multiply(A=@n6, B=@n7)) {speedX=-1}) {samplerType=SAMPLERTYPE_Normal, texture=/Game/Global/Textures/Default/Noise/T_Normal_Noise01}
  n9 = Multiply(A=@n8.R) {constB=0.15}
  n10 = TextureSample(UVs=Panner(Coordinate=Add(A=@n5, B=@n9) {constB=0}) {speedX=-1}) {
   samplerSource=SSM_Wrap_WorldGroupSettings
   samplerType=SAMPLERTYPE_Grayscale
   texture=&T_Smoke_Trail_Soft
  }
  n11 = TextureCoordinate()
  n12 = Panner(Coordinate=Add(A=TextureCoordinate() {uTiling=0.8, vTiling=1.5}, B=Multiply(A=@n8.R) {constB=0.08})) {speedX=-0.4}
  n13 = Multiply(
   A=Power(Base=ComponentMask(None=@n11) {r=true}) {constExponent=1}
   B=LinearInterpolate(Alpha=TextureSample(UVs=@n12) {texture=/BRCosmetics/Weapons/FORT_Melee/Pickaxe_SoapPocket/FX/T_Pickaxe_SoapPocket_Noise}.R) {constA=0.2}
  )
  n14 = Saturate(None=Power(Base=Multiply(A=@n13) {constB=10}) {constExponent=1.5})
  n15 = Multiply(A=Constant3Vector() {constant=<rgba 0.8958 0.55 0.0448 1>}, B=Multiply(A=@n10.R, B=@n14))
  n16 = TextureCoordinate() {uTiling=0.5, vTiling=2}
  n17 = TextureSample(UVs=Panner(Coordinate=Add(A=@n16, B=@n9)) {speedX=-0.3}) {
   samplerSource=SSM_Wrap_WorldGroupSettings
   samplerType=SAMPLERTYPE_LinearGrayscale
   texture=/Game/Global/Textures/Default/Caustics/T_Nosie_Caustics_Grayscale
  }
  n18 = TextureCoordinate() {uTiling=0.3}
  n19 = Add(
   A=LinearInterpolate(Alpha=@n17.R) {constA=0.2, constB=0.8}
   B=TextureSample(UVs=Panner(Coordinate=@n18) {speedX=-0.5}) {samplerType=SAMPLERTYPE_LinearColor, texture=/Game/Effects/Fort_Effects/Emotes/UltraLight/Texture/T_Plasma_D_C}.R
  )
  n20 = ComponentMask(None=TextureCoordinate()) {r=true}
  n21 = Saturate(None=Power(Base=OneMinus(None=@n20), Exp=ScalarParameter() {defaultValue=1.5, parameterName=TailMask}) {constExponent=1})
  n22 = TextureCoordinate()
  n23 = Power(Base=Sine(None=ComponentMask(None=@n22) {g=true}) {period=2}, Exp=ScalarParameter() {defaultValue=5, parameterName=SideAlpha})
  n24 = ScalarParameter() {defaultValue=50, parameterName=EdgeFade}
  n25 = Multiply(A=Multiply(A=Multiply(A=@n19, B=@n21), B=@n14), B=Saturate(None=Multiply(A=@n23, B=@n24) {constB=2}))
  n26 = Multiply(A=Constant3Vector() {constant=<rgba 1 0.42 0.05 1>}, B=@n25)
  MaterialFunctionCall(Emissive=Multiply(A=VectorParameter() {parameterName=FinalColor, defaultValue=<rgba 1 1 1 1>}.RGB, B=Add(A=@n15, B=@n26))) {materialFunction=&MF_FX_EmissiveAdjustments}
  n27 = ScalarParameter() {defaultValue=3, parameterName=FadeDistance}
  Saturate(None=Multiply(A=ScalarParameter() {defaultValue=2, parameterName=MainAlpha}, B=DepthFade(Opacity=@n25, FadeDistance=@n27) {fadeDistanceDefault=5}) {constA=1.3})
 ]
}
