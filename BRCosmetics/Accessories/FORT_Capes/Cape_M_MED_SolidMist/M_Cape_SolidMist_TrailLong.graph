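# Translucent, unlit, two-sided material for the long mist trail on the Solid Mist cape cosmetic.
# SplineThicken shapes width along U, while world position offset adds masked sine waves, vertex-color-driven side and camera offsets, and platform-switched intensity.
# Two panned fire-trail samples, warped by opposing noise, combine with noise-distorted gradients and curve-atlas ramps for tinted emissive and opacity.
# Vertex red varies motion, Grow controls erosion, and DepthFade softens intersections.
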Material /BRCosmetics/Accessories/FORT_Capes/Cape_M_MED_SolidMist/M_Cape_SolidMist_TrailLong
graph {
 properties {
  bAutomaticallySetUsageInEditor=false
  bUsedWithMeshDeformer=true
  bUsedWithSkeletalMesh=true
  blendMode=BLEND_Translucent
  shadingModel=MSM_Unlit
  twoSided=true
  usedShadingModels=Unlit
 }
 nodes [
  n1 = Constant() {r=0.5}
  n2 = Constant() {r=0.4}
  n3 = Constant() {r=1.5}
  n4 = Constant() {r=1.2}
  n5 = ComponentMask(None=TextureCoordinate()) {r=true}
  n6 = LinearInterpolate(Alpha=Power(Base=@n5) {constExponent=1.2})
  n7 = Saturate(None=Multiply(A=OneMinus(None=@n5)) {constB=2})
  n8 = Add(A=TextureCoordinate(), B=Constant2Vector() {r=0.28})
  n9 = Multiply(A=ComponentMask(None=@n8) {r=true}) {constB=1.2}
  n10 = Multiply(A=OneMinus(None=ComponentMask(None=@n8) {r=true})) {constB=1.5}
  n11 = Saturate(None=Add(A=Power(Base=@n9) {constExponent=6}, B=Power(Base=@n10) {constExponent=8}))
  n12 = LinearInterpolate(
   A=MaterialFunctionCall("Xb1,PS4"=@n1, "PC-High&&Epic,XSX,PS5"=@n1, "PC-Perf,Android"=@n2, Switch=@n2, "PC-Low(NotPerf)&&Medium"=@n1) {materialFunction=&PlatformSwitch_1D}.Result
   B=MaterialFunctionCall("Xb1,PS4"=@n3, "PC-High&&Epic,XSX,PS5"=@n3, "PC-Perf,Android"=@n4, Switch=@n4, "PC-Low(NotPerf)&&Medium"=@n3) {materialFunction=&PlatformSwitch_1D}.Result
   Alpha=Multiply(A=Multiply(A=@n6, B=@n7) {constA=1}, B=LinearInterpolate(Alpha=@n11) {constA=-0.3})
  ) {constA=0.1, constB=2.5}
  n13 = Constant() {r=12}
  n14 = MaterialFunctionCall(WidthBase=Multiply(A=@n12, B=@n13) {constA=1}, WidthTip=Multiply(A=@n12, B=@n13) {constA=1}, Expand U or V UV Channel=StaticBool()) {materialFunction=&SplineThicken}
  n15 = TextureCoordinate()
  n16 = LinearInterpolate(Alpha=VertexColor().R) {constA=0.5, constB=1.2}
  n17 = Power(Base=ComponentMask(None=Multiply(A=@n15, B=@n16)) {r=true}) {constExponent=0.5}
  n18 = Multiply(A=Add(A=Time(), B=VertexColor().R) {constB=0}) {constB=-0.5}
  n19 = LinearInterpolate(Alpha=Sine(None=Add(A=@n17, B=@n18)) {period=0.7}) {constA=-0.5, constB=0.5}
  n20 = Transform(None=VectorParameter() {parameterName=WPODirection, defaultValue=<rgba 0 0 1 1>}.RGB)
  n21 = VertexColor()
  n22 = ComponentMask(None=Multiply(A=TextureCoordinate(), B=LinearInterpolate(Alpha=@n21.R) {constA=0.5, constB=1.2})) {r=true}
  n23 = Add(A=Time(), B=VertexColor().R) {constB=0}
  n24 = Sine(None=Add(A=Power(Base=@n22) {constExponent=0.5}, B=Multiply(A=@n23) {constB=-0.5}) {constB=0}) {period=0.7}
  n25 = Transform(None=VectorParameter() {parameterName=WPODirection02, defaultValue=<rgba 0 0.6 0 1>}.RGB) {transformSourceType=TRANSFORMSOURCE_Local}
  n26 = Add(A=Multiply(A=Multiply(A=@n19, B=@n20)), B=Multiply(A=Multiply(A=@n24, B=@n25))) {constB=0}
  n27 = ScalarParameter() {defaultValue=8, parameterName=WPOIntensity}
  n28 = ScalarParameter() {defaultValue=5, parameterName=WPOIntensityLOD1}
  n29 = MaterialFunctionCall("Xb1,PS4"=@n27, "PC-High&&Epic,XSX,PS5"=@n27, "PC-Perf,Android"=@n28, Switch=@n28, "PC-Low(NotPerf)&&Medium"=@n27) {materialFunction=&PlatformSwitch_1D}
  n30 = TextureCoordinate()
  n31 = ComponentMask(None=@n30) {r=true}
  n32 = ScalarParameter() {defaultValue=1, parameterName=TopMask}
  n33 = OneMinus(None=ComponentMask(None=TextureCoordinate()) {r=true})
  n34 = Multiply(A=LinearInterpolate(Alpha=Power(Base=@n33) {constExponent=0.5}) {constA=0.1})
  n35 = Multiply(A=Multiply(A=Power(Base=@n31, Exp=@n32) {constExponent=0.5}, B=Saturate(None=@n34))) {constB=1.3}
  n36 = Add(A=Multiply(A=Multiply(A=@n26, B=@n29.Result), B=Saturate(None=@n35))) {constB=0}
  n37 = Add(
   A=Add(A=Add(A=@n14.WorldPositionOffset, B=@n36) {constB=0}) {constB=0}
   B=Transform(None=VectorParameter() {parameterName=Offset02, defaultValue=<rgba 0 -2 0 1>}.RGB) {transformSourceType=TRANSFORMSOURCE_Local}
  )
  n38 = Multiply(
   A=Transform(None=VectorParameter() {parameterName=SideMeshOffset, defaultValue=<rgba 0 -2 0 1>}.RGB) {transformSourceType=TRANSFORMSOURCE_Local}
   B=VertexColor().G
  )
  n39 = VertexColor()
  n40 = MaterialFunctionCall(Offset Amount=ScalarParameter() {defaultValue=3, parameterName="SideOffset Amount02 (S)"}) {materialFunction=/Engine/Functions/Engine_MaterialFunctions02/WorldPositionOffset/CameraOffset}
  Add(
   A=Add(A=Add(A=@n37, B=@n38), B=Multiply(A=@n39.G, B=@n40.Camera Offset))
   B=MaterialFunctionCall(Offset Amount=ScalarParameter() {defaultValue=7, parameterName="Offset Amount (S)"}) {materialFunction=&CameraOffset}.Camera Offset
  )
  Add(A=@n30, B=Constant2Vector() {r=-0.05})
  Constant() {r=2}
  n41 = VertexColor()
  n42 = Constant()
  n43 = Add(A=NamedRerouteUsage(), B=Add(A=TextureCoordinate(), B=AppendVector(A=@n41.R, B=@n42)))
  n44 = AppendVector(A=Constant() {r=3}, B=Constant() {r=1})
  n45 = MaterialFunctionCall(UV Channel=MaterialFunctionCall(UVs=Multiply(A=@n43, B=@n44), Texture Scale=Constant2Vector() {g=1.2, r=-1}) {materialFunction=&ScaleUVsByCenter}.UVs) {materialFunction=&LinearGradient}
  n46 = AppendVector(A=OneMinus(None=@n45.UGradient), B=@n45.VGradient)
  n47 = Multiply(A=Time(), B=Multiply(A=Constant() {r=2}, B=VertexColor().R))
  n48 = MaterialFunctionCall(UV Channel=MaterialFunctionCall(UVs=Multiply(A=@n43), Texture Scale=Constant2Vector() {g=1.5, r=-0.7}) {materialFunction=&ScaleUVsByCenter}.UVs) {materialFunction=&LinearGradient}
  n49 = AppendVector(A=OneMinus(None=@n48.UGradient), B=@n48.VGradient)
  n50 = Multiply(
   A=TextureSample(UVs=Panner(Coordinate=@n46, Time=@n47) {speedX=-0.5}) {texture=/Game/Effects/Fort_Effects/Textures/Fire/T_Stylized_Fire_Trail}.R
   B=TextureSample(UVs=Panner(Coordinate=@n49, Time=@n47) {speedX=-0.4}) {texture=&T_Stylized_Fire_Trail}.R
  )
  n51 = Power(
   Base=MaterialFunctionCall() {materialFunction=&LinearGradient}.UGradient
   Exp=Constant() {r=0.6}
  ) {constExponent=0.5}
  n52 = Constant() {r=0.5}
  n53 = Saturate(None=Subtract(A=SphereMask(A=@n51, B=@n52) {attenuationRadius=0.5, hardnessPercent=0}, B=ScalarParameter() {defaultValue=0.2}) {constB=0.2})
  n54 = OneMinus(None=@n53)
  n55 = Multiply(A=Power(Base=@n50, Exp=Constant() {r=1}), B=@n54)
  n56 = Constant() {r=0.8}
  n57 = VertexColor()
  n58 = Constant() {r=1.1}
  n59 = MaterialFunctionCall() {materialFunction=/Engine/Functions/Engine_MaterialFunctions02/WorldPositionOffset/ObjectScale}
  n60 = Multiply(A=TextureCoordinate() {uTiling=2}, B=AppendVector(A=Multiply(A=@n56, B=@n57.R), B=Multiply(A=@n58, B=@n59.Scale X)))
  n61 = ScalarParameter() {defaultValue=1, parameterName=PanSpeed}
  n62 = VertexColor()
  n63 = Multiply(A=Constant2Vector() {r=-1}, B=Multiply(A=Multiply(A=@n61, B=@n62.R), B=Constant() {r=1}))
  n64 = Sine(None=Divide(A=MaterialFunctionCall() {materialFunction=&LinearGradient}.VGradient))
  n65 = Divide(
   A=TextureSample(UVs=Panner(Coordinate=@n60, Speed=@n63) {speedX=-0.02, speedY=-0.3}) {samplerType=SAMPLERTYPE_Normal, texture=/BRCosmetics/Effects/Fort_Effects/Emotes/UltraLight/Texture/T_DirectionalNoise_Normal}.RGB
   B=LinearInterpolate(A=Constant() {r=2}, B=Constant() {r=7}, Alpha=Saturate(None=@n64)) {constA=1, constB=4}
  )
  n66 = MaterialFunctionCall(UV Channel=Add(A=TextureCoordinate(), B=ComponentMask(None=@n65) {g=true, r=true})) {materialFunction=&LinearGradient}
  n67 = Saturate(None=Multiply(A=@n66.UGradient) {constB=3.7})
  n68 = Multiply(
   A=MaterialFunctionCall() {materialFunction=&LinearGradient}.VGradient
   B=LinearInterpolate(Alpha=DynamicParameter() {paramNames [Grow, Emissive, Width, SpiralWPODirection]}.Grow) {constA=10, constB=0}
  )
  n69 = Saturate(None=Subtract(A=Multiply(A=@n55, B=@n67), B=@n68))
  n70 = CurveAtlasRowParameter(CurveTime=@n66.VGradient) {
   atlas=/Game/Effects/Fort_Effects/GameplayPulginsAssets/Demon/CA_GoldenFire
   curve=/Game/Effects/Fort_Effects/GameplayPulginsAssets/Demon/Curve_GoldenFire01
   parameterName=Param_1
  }
  n71 = Multiply(A=@n70.A)
  n72 = MaterialFunctionCall() {materialFunction=&LinearGradient}
  n73 = Saturate(None=Sine(None=Divide(A=@n72.UGradient)))
  n74 = LinearInterpolate(
   A=Constant() {r=5}
   B=Constant() {r=100}
   Alpha=MaterialFunctionCall(Distance=Constant() {r=2000}) {materialFunction=&Near_Camera_Fade}.Result
  ) {constA=2, constB=100}
  n75 = OneMinus(None=Saturate(None=Multiply(A=@n73, B=@n74) {constB=2}))
  n76 = Subtract(A=Saturate(None=Subtract(A=@n71, B=@n75) {constB=0.1}), B=@n68) {constB=0.1}
  n77 = Multiply(A=Power(Base=Saturate(None=@n76), Exp=Constant() {r=2}), B=Constant() {r=2.5})
  n78 = ComponentMask(None=TextureCoordinate()) {r=true}
  n79 = Saturate(None=Power(Base=Multiply(A=@n78) {constB=5}))
  n80 = OneMinus(None=@n78)
  n81 = Saturate(None=Power(Base=Multiply(A=@n80) {constB=10}))
  DepthFade(Opacity=Multiply(A=Add(A=@n69, B=@n77), B=Multiply(A=@n79, B=@n81)))
  n82 = CurveAtlasRowParameter(CurveTime=@n50) {
   atlas=&CA_GoldenFire
   curve=/Game/Effects/Fort_Effects/GameplayPulginsAssets/Demon/Curve_GoldenFire03
   defaultValue=2
   parameterName=Param_5
  }
  n83 = Multiply(A=Multiply(A=Multiply(A=@n82) {constB=2}, B=@n54), B=@n67)
  n84 = LinearInterpolate(A=@n70, B=Divide(A=@n70, B=ScalarParameter() {defaultValue=1, parameterName=Param_2}), Alpha=@n53)
  n85 = CurveAtlasRowParameter(CurveTime=@n66.VGradient) {
   atlas=&CA_GoldenFire
   curve=/Game/Effects/Fort_Effects/GameplayPulginsAssets/Demon/Curve_GoldenFire02
   defaultValue=1
   parameterName=Param_2
  }
  n86 = OneMinus(None=@n67)
  n87 = Multiply(
   A=Desaturation(None=LinearInterpolate(A=@n84, B=@n85, Alpha=@n86), Fraction=Constant() {r=0.6})
   B=Constant3Vector() {constant=<rgba 2 1.5263 0.2 1>}
  )
  MaterialFunctionCall(Emissive=Multiply(A=VectorParameter() {parameterName=FinalColor, defaultValue=<rgba 1 1 1 1>}.RGB, B=Add(A=@n83, B=@n87))) {materialFunction=&MF_FX_EmissiveAdjustments}
  MaterialFunctionCall() {materialFunction=&DebugTimeSine}
  n88 = VertexColor()
  n89 = Constant()
  n90 = Constant() {r=4}
  n91 = Constant() {r=1.5}
  n92 = Multiply(A=Add(A=TextureCoordinate(), B=AppendVector(A=@n88.R, B=@n89)), B=Divide(A=AppendVector(A=@n90, B=@n91), B=Constant() {r=1}))
  n93 = Multiply(A=Constant() {r=-5}, B=AppendVector(A=VertexColor().R, B=Constant()))
  n94 = Constant() {r=3}
  n95 = Constant() {r=3}
  n96 = Panner(Coordinate=@n92, Speed=Divide(A=AppendVector(A=@n93, B=@n94), B=@n95)) {speedX=0.2, speedY=0.2}
  n97 = Round(None=MaterialFunctionCall() {materialFunction=&LinearGradient}.VGradient)
  n98 = Multiply(
   A=ComponentMask(None=TextureSample(UVs=@n96) {samplerType=SAMPLERTYPE_Normal, texture=/Game/Global/Textures/Default/Noise/T_Normal_Noise01}.RGB) {g=true, r=true}
   B=OneMinus(None=@n97)
  )
  n99 = Multiply(A=@n94) {constB=-1}
  n100 = Panner(Coordinate=@n92, Speed=Divide(A=AppendVector(A=@n93, B=@n99), B=@n95)) {speedX=0.2, speedY=-0.2}
  n101 = Multiply(
   A=ComponentMask(None=TextureSample(UVs=@n100) {samplerType=SAMPLERTYPE_Normal, texture=&T_Normal_Noise01}.RGB) {g=true, r=true}
   B=@n97
  )
  NamedRerouteDeclaration(None=Divide(A=Add(A=@n98, B=@n101), B=Constant() {r=80})) {name=UVDistortionOpposite, nodeColor=<rgba 0 0.7656 1 1>}
 ]
}
